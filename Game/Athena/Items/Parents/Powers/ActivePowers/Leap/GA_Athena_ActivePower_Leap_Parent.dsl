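; Parent gameplay ability for Fortnite Athena leap powers, with overridable hooks for child abilities.
; A cast montage notify commits and leaps, first applying fall settings and optionally ejecting the player from vehicles or mounts.
; Launch uses movement mode extensions or root motion for directional and arc leaps, optionally weakened by predictive obstacle targeting.
; Server timed capsule traces apply effects mid-leap; landing or water entry triggers cues, targeting and optional movement blocking.
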(event Custom|ApplyGuidedMotionMME
  (bind _optionalmmefloats (Variables|PowerLeap|MME|GetOptionalMMEFloats))
  ;[0.07118 0.002676 0.15 0.5] ApplyGuidedMotionMME: Setting up as an event so it can be overridden.
  (Utilities|Array|SetArrayElem _optionalmmefloats 0 (Variables|PowerCharge|GetChargeMultiplier) true)
  (Utilities|Array|SetArrayElem _optionalmmefloats 1 (Variables|PowerLeap|PredictiveLaunchVelocity|RuntimeVars|GetPredictiveLaunchVelocityMult) true)
  (bind _returnvalue (Game|ApplyGuidedMotionMME (Variables|Default|GetPlayerPawn) (Variables|PowerLeap|MME|GetGuidedMotionLeapMME) 0 0 (Utilities|Struct|MakeGuidedMotionOptionalValues :OptionalFloats _optionalmmefloats)))
  (Utilities|IsValid _returnvalue
    (:Is%20Not%20Valid
      (Fortnite|Utility|FortBlueprintLog (Utilities|String|Append (Utilities|GetDisplayName (Variables|Default|GetPlayerPawn)) " failed to activate MME with tag ")))))

(event Custom|LaunchPlayer
  ;[1 1 1] Calculate the leap based on the type of leap we are doing after apply root motion based on leap type.
  (if (Variables|PowerLeap|GetPlayCastAnimationOnLaunch)
    (CallFunction|PlayCastAnimation)
    (else
      _))
  (Variables|PowerLeap|PredictiveLaunchVelocity|RuntimeVars|SetPredictiveLaunchVelocityMult 1.0)
  ;;[1 1 1] Predictive Launch - reduces launch strength when hitting objects that would not be destroyed by leap
  (if (ScalableFloat|GetValueasBool (Variables|PowerLeap|PredictiveLaunchVelocity|GetRow_UsePredictiveLaunchVelocity))
    (Utilities|IsValid (Variables|PowerLeap|PredictiveLaunchVelocity|GetTargetingPreset_PredictiveLaunchVelocity)
      (:Is%20Valid
        (bind (_asynctaskproxy _targetinghandle) (Class|AsyncActionPerformTargeting|PerformTargetingAsyncAction (Variables|Default|GetPlayerPawn) (Variables|PowerLeap|PredictiveLaunchVelocity|GetTargetingPreset_PredictiveLaunchVelocity) true)
          (:then
            (Utilities|IsValid _asynctaskproxy
              (:Is%20Valid
                (bind _output_get (Variables|PowerLeap|PredictiveLaunchVelocity|RuntimeVars|SetTargetingHandle_PredictiveLaunch (Targeting|GetTargetingHandle _asynctaskproxy)))
                (TargetingSystem|TargetingTypes|SetTargetingSourceContext _output_get (Utilities|Struct|MakeTargetingSourceContext (Variables|Default|GetPlayerPawn) 0 "0.000000,0.000000,0.000000" "None" self)))))
          (:Targeted
            (Default|CalculatePredictiveLaunchMultiplier (TargetingSystem|TargetingResults|GetTargetingResults (GameInstanceSubsystems|GetTargetingSubsystem) _targetinghandle)))))
      (:Is%20Not%20Valid
        _))
    (if (Variables|PowerLeap|MME|GetbLeapWithMME)
      (Utilities|IsServer
        (:true
          _)
        (:false
          (if (ScalableFloat|GetValueasBoolwithContext (Variables|PowerLeap|MME|GetUseGuidedMotionLeap))
            (CallFunction|ApplyGuidedMotionMME)
            (else
              (bind _returnvalue (CallFunction|GetLaunchVelocity))
              (bind _output_get_2 (Variables|PowerLeap|UpwardOrDownward|SetCalculatedLaunchVelocity _returnvalue))
              (bind _returnvalue_2 (Class|FortMovementModeExtLeap|TryActivateLeapMMEGetRef (Variables|Default|GetPlayerPawn) (Variables|PowerLeap|MME|GetLeapMMEClass) (Utilities|Struct|MakeFortMovementModeLeapCreationData _output_get_2 (CallFunction|GetLeapDuration))))
              (Utilities|IsValid _returnvalue_2
                (:Is%20Valid
                  _)
                (:Is%20Not%20Valid
                  (Fortnite|Utility|FortBlueprintLog (Utilities|String|Append (Utilities|GetDisplayName (Variables|Default|GetPlayerPawn)) " failed to activate MME with tag "))))))))
      (else
        (Character|StopJumping (Variables|Default|GetPlayerPawn))
        (switch Utilities|FlowControl|Switch|SwitchonAthena_Power_Leap_Type (Variables|PowerLeap|GetLeapType)
          (:NewEnumerator0
            _)
          (:NewEnumerator3
            _)
          (:NewEnumerator2
            (bind _returnvalue_3 (CallFunction|GetLaunchVelocity))
            (bind _output_get_3 (Variables|PowerLeap|UpwardOrDownward|SetCalculatedLaunchVelocity _returnvalue_3))
            (bind _asynctaskproxy_2 (Ability|Tasks|ApplyRootMotionConstantForce "None" (Math|Vector|Normalize _output_get_3) (Math|Vector|VectorLength _output_get_3) (CallFunction|GetLeapDuration) false (Variables|PowerLeap|UpwardOrDownward|GetCurve_StrengthOverTime) (Variables|PowerLeap|VelocityOnFinish|GetVelocityOnFinish) (CallFunction|GetSetVelocityonFinish) (CallFunction|GetClampVelocityonFinish) (Variables|PowerLeap|UpwardOrDownward|GetGravityOnLeap))
              (:then
                (Variables|PowerLeap|SetLeapConstTask _asynctaskproxy_2)
                (CallFunction|GiveTillLandFallImmunity)
                (Variables|PowerLeap|RuntimeVars|SetLaunchStarted true)
                (CallFunction|OnLaunchStarted) ; Can be overridden in children for post launch behavior.
              )
              (:OnFinish
                (CallFunction|OnLeapFinish))))
          (:NewEnumerator1
            _))
        (bind _targetlocation (CallFunction|SelectArcLocation))
        (bind _output_get_4 (Variables|PowerLeap|Arc|SetCalculatedArcTargetLocation _targetlocation))
        (bind _asynctaskproxy_3 (Ability|Tasks|ApplyRootMotionMoveToForce "None" _output_get_4 (CallFunction|GetLeapDuration) false "MOVE_None" false (Variables|PowerLeap|Arc|GetCurve_LeapPathOffset) (Variables|PowerLeap|VelocityOnFinish|GetVelocityOnFinish) (CallFunction|GetSetVelocityonFinish) (CallFunction|GetClampVelocityonFinish))
          (:then
            (Variables|PowerLeap|SetLeapMoveToTask _asynctaskproxy_3)
            (CallFunction|GiveTillLandFallImmunity))
          (:OnTimedOut
            (CallFunction|OnLeapFinish)))))
    (else
      _)))

(event Custom|StartEjectionLeap
  ;[1 1 1] Start Ejection Leap
  (Variables|PowerLeap|Ejection|RunTimeVars|SetDurationToIgnoreCollision (CallFunction|EjectionLeapGetDurationtoIgnoreCollisionwithActorEjectedFrom))
  (if (> (Variables|PowerLeap|Ejection|RunTimeVars|GetDurationToIgnoreCollision) 0)
    (Collision|AddIgnoredActor (Variables|Default|GetPlayerPawn) (Variables|PowerLeap|Ejection|RunTimeVars|GetEjectionLeap_ActorEjectedFrom) (Variables|PowerLeap|Ejection|RunTimeVars|GetDurationToIgnoreCollision))
    (else
      _))
  (bind _asynctaskproxy (Ability|Tasks|ApplyRootMotionConstantForce "None" (Math|Vector|Normalize (CallFunction|GetEjectionLeapVelocity)) (Math|Vector|VectorLength (CallFunction|GetEjectionLeapVelocity)) (ScalableFloat|GetValueatLevel (Variables|PowerLeap|Ejection|GetRow_EjectionLeap_Duration)) (Variables|PowerLeap|Ejection|GetEjectionLeap_IsAdditive) (Variables|PowerLeap|Ejection|GetEjectionLeap_StrengthOverTimeCurve) (Variables|PowerLeap|Ejection|GetEjectionLeap_VelocityOnFinishMode) (CallFunction|EjectionLeapGetSetVelocityonFinish) (Default|EjectionLeapGetClampVelocityonFinish) (Variables|PowerLeap|Ejection|GetEjectionLeap_EnableGravity))
    (:then
      (Variables|PowerLeap|Ejection|RunTimeVars|SetEjectLeap_Task _asynctaskproxy))
    (:OnFinish
      (CallFunction|EjectionLeapFinished))))

(event EventSetWaterBinding
  ;[1 1 1] Player Enter Water - Should Land be triggered when the player enters water.
  (Ability|Tasks|WaitGameplayTagAdd (Variables|PowerActivation|GetTag_SurfaceSwimming)
    (:then
      (Variables|PowerLeap|SetLandedInWater false)
      (|CallParentFunction))
    (:Added
      (if (Variables|PowerLeap|GetTriggerLandOnEnterWater)
        (Variables|PowerLeap|SetLandedInWater true)
        (CallFunction|OnLand (Collision|MakeHitResult false))))))

(event EventPlayerEnterWater
  (params (FortPlayerPawn :type Fort%20Player%20Pawn%20Object%20Reference))
  ;[1 1 1] Player Enter Water - Should Land be triggered when the player enters water.
  (|CallParentFunction FortPlayerPawn)
  (if (Variables|PowerLeap|GetTriggerLandOnEnterWater)
    (Variables|PowerLeap|SetLandedInWater true)
    (CallFunction|OnLand (Collision|MakeHitResult false))))

(event Custom|Leap
  ;[1 1 1] Leap - Play leap GC and Add Looping, if the leap uses land bind to the pawn landing. Then Call the launch event and check if we should start targeting while leaping
  (CallFunction|SetFallSettings)
  (if (ScalableFloat|GetValueasBool (Variables|PowerLeap|GetRow_ShouldEndZiplining))
    (Zipline|EndZiplining (Variables|Default|GetPlayerPawn) false false)
    (else
      _))
  (if (Variables|PowerTargetWhileLeaping|GetDoesTargetingWhileLeaping)
    (CallFunction|StartTargetingWhileLeaping)
    (else
      _))
  (CallFunction|CheckforEjection)
  (if (Variables|PowerLeap|Ejection|RunTimeVars|GetWasEjected)
    (CallFunction|StartEjectionLeap)
    (else
      (CallFunction|PlayBurstGC (Variables|PowerGameplayCues|GetGC_Leap) (Variables|PowerGameplayCues|GetLeapGCUsesParams) (CallFunction|LeapGCParams))
      (CallFunction|AddGCLoop (Variables|PowerGameplayCues|GetGC_LeapingLoop) (Variables|PowerGameplayCues|GetLeapingLoopGCUsesParams) (CallFunction|LeapingLoopGCParams) true)
      (if (Variables|PowerLeap|GetHasLand)
        (EventDispatchers|BindEventtoOnPawnLanded (Variables|Default|GetPlayerPawn) (EventDispatchers|CreateEvent :!event Custom|OnLand))
        (else
          _))
      (CallFunction|LaunchPlayer))))

(event Custom|EjectionLeapFinished
  ;[1 1 1] Leap - Play leap GC and Add Looping, if the leap uses land bind to the pawn landing. Then Call the launch event and check if we should start targeting while leaping
  (CallFunction|PlayBurstGC (Variables|PowerGameplayCues|GetGC_Leap) (Variables|PowerGameplayCues|GetLeapGCUsesParams) (CallFunction|LeapGCParams))
  (CallFunction|AddGCLoop (Variables|PowerGameplayCues|GetGC_LeapingLoop) (Variables|PowerGameplayCues|GetLeapingLoopGCUsesParams) (CallFunction|LeapingLoopGCParams) true)
  (if (Variables|PowerLeap|GetHasLand)
    (EventDispatchers|BindEventtoOnPawnLanded (Variables|Default|GetPlayerPawn) (EventDispatchers|CreateEvent :!event Custom|OnLand))
    (else
      _))
  (CallFunction|LaunchPlayer))

(event Custom|OnLeapFinish
  ;[1 1 1] Check if the Ability should end on leap finish.
  (if (Variables|PowerLeap|GetEndAbilityOnLeapFinish)
    (Ability|EndAbility)))

(event Custom|OnLand
  (params (Hit :type Hit%20Result%20Structure%20%28by%20ref%29))
  ;[1 1 1] On Land - If we bound to land jump the montage to the land section, play the land burst and do targeting if applicable.
  (Utilities|FlowControl|DoOnce false
    (:Completed
      (GameplayTasks|EndTask (Variables|PowerLeap|GetLeapMoveToTask))
      (Utilities|Time|ClearandInvalidateTimerbyHandle (Variables|PowerTargetWhileLeaping|GetTimer_TargetWhileLeaping))
      (Ability|RemoveGameplayCueFromOwner (Variables|PowerGameplayCues|GetGC_LeapingLoop))
      (EventDispatchers|UnbindEventfromOnPawnLanded (Variables|Default|GetPlayerPawn) (delegate Custom|OnLand))
      (Ability|Animation|MontageJumptoSection (Variables|PowerLeap|GetCastMontageLandSection))
      (CallFunction|PlayBurstGC (Variables|PowerGameplayCues|GetGC_Land) (Variables|PowerGameplayCues|GetLandGCUsesParams) (CallFunction|LandGCParams))
      (if (Variables|PowerLeap|GetTargetOnLand)
        (CallFunction|TargetLand)
        (else
          _))
      (if (Variables|PowerLeap|GetBlockPlayerMovementForLand)
        (CallFunction|SetBlockMovement (Variables|Default|GetPlayerPawn) true (Variables|PowerLeap|GetBlockMovementOnLandHasDuration) (Variables|PowerLeap|GetBlockMovementOnLandDuration))
        (Utilities|FlowControl|Sequence
          (:then_0
            (if (not (Variables|PowerLeap|GetLandedInWater))
              (Components|Movement|StopMovementImmediately (Variables|/Script/Engine.Character|GetCharacterMovement (Variables|Default|GetPlayerPawn)))))
          (:then_1
            (if (Variables|PowerLeap|GetEndAbilityOnLandFinish)
              (if (Variables|PowerLeap|GetBlockMovementOnLandHasDuration)
                (Utilities|FlowControl|Delay (Variables|PowerLeap|GetBlockMovementOnLandDuration))
                (else
                  _))
              (Ability|EndAbility))))))))

(event EventCastMontageBlendOut
  ;[1 1 1] MontageBlendOut - End Ability
  (|CallParentFunction)
  (Ability|EndAbility))

(event Custom|ResetLandDoOnce
  ;[1 1 1] On Land - If we bound to land jump the montage to the land section, play the land burst and do targeting if applicable.
  (Utilities|FlowControl|DoOnce false
    (:Completed
      (GameplayTasks|EndTask (Variables|PowerLeap|GetLeapMoveToTask))
      (Utilities|Time|ClearandInvalidateTimerbyHandle (Variables|PowerTargetWhileLeaping|GetTimer_TargetWhileLeaping))
      (Ability|RemoveGameplayCueFromOwner (Variables|PowerGameplayCues|GetGC_LeapingLoop))
      (EventDispatchers|UnbindEventfromOnPawnLanded (Variables|Default|GetPlayerPawn) (delegate Custom|OnLand))
      (Ability|Animation|MontageJumptoSection (Variables|PowerLeap|GetCastMontageLandSection))
      (CallFunction|PlayBurstGC (Variables|PowerGameplayCues|GetGC_Land) (Variables|PowerGameplayCues|GetLandGCUsesParams) (CallFunction|LandGCParams))
      (if (Variables|PowerLeap|GetTargetOnLand)
        (CallFunction|TargetLand)
        (else
          _))
      (if (Variables|PowerLeap|GetBlockPlayerMovementForLand)
        (CallFunction|SetBlockMovement (Variables|Default|GetPlayerPawn) true (Variables|PowerLeap|GetBlockMovementOnLandHasDuration) (Variables|PowerLeap|GetBlockMovementOnLandDuration))
        (Utilities|FlowControl|Sequence
          (:then_0
            (if (not (Variables|PowerLeap|GetLandedInWater))
              (Components|Movement|StopMovementImmediately (Variables|/Script/Engine.Character|GetCharacterMovement (Variables|Default|GetPlayerPawn)))))
          (:then_1
            (if (Variables|PowerLeap|GetEndAbilityOnLandFinish)
              (if (Variables|PowerLeap|GetBlockMovementOnLandHasDuration)
                (Utilities|FlowControl|Delay (Variables|PowerLeap|GetBlockMovementOnLandDuration))
                (else
                  _))
              (Ability|EndAbility))))))))

(event EventCastMontageNotifyEndAbility
  ;[1 1 1] MontageEndNotify - End Ability
  (|CallParentFunction)
  (Ability|EndAbility))

(event EventCastMontageCancelledOrInterrupted
  ;[1 1 1] Montage Cancelled or Interrupted - end ability
  (|CallParentFunction)
  (CallFunction|CleanUponLand)
  (Ability|EndAbility))

(event Custom|StartTargetingWhileLeaping
  ;[1 1 1] Target While Leaping - Start a timer that when ticks projects a capsule in the direction of the player's velocity applying effects to any actors that are hit by the trace.
  (if (and (Networking|IsServer) (or (not (Variables|PowerTargetWhileLeaping|GetTargetWhileLeapRequiresChargeThreshold)) (and (Variables|PowerTargetWhileLeaping|GetTargetWhileLeapRequiresChargeThreshold) (>= (Variables|PowerCharge|GetChargePercent) (ScalableFloat|GetValueatLevel (Variables|PowerTargetWhileLeaping|GetTargetWhileLeapChargeThreshold))))))
    (bind _returnvalue (Utilities|Time|SetTimerbyEvent (delegate Custom|TargetWhileLeaping) (Variables|PowerTargetWhileLeaping|GetTargetWhileLeapingInterval) true))
    (Variables|PowerTargetWhileLeaping|SetTimer_TargetWhileLeaping _returnvalue)
    (Utilities|Array|Add (Variables|PowerTargetWhileLeaping|GetWhileLeapingHitActors) (Variables|Default|GetPlayerPawn))
    (CallFunction|TargetWhileLeaping)))

(event Custom|CleanUpOnLand
  ;[1 1 1] Clean up On Land event if montage was cancelled or interrupted.
  (GameplayTasks|EndTask (Variables|PowerLeap|GetLeapMoveToTask))
  (Utilities|Time|ClearandInvalidateTimerbyHandle (Variables|PowerTargetWhileLeaping|GetTimer_TargetWhileLeaping))
  (Ability|RemoveGameplayCueFromOwner (Variables|PowerGameplayCues|GetGC_LeapingLoop))
  (EventDispatchers|UnbindEventfromOnPawnLanded (Variables|Default|GetPlayerPawn) (delegate Custom|OnLand)))

(event EventPowerCommitFailed
  ;[1 1 1] Commit Failed - End Ability
  (|CallParentFunction)
  (Ability|EndAbility))

(event Custom|TargetWhileLeaping
  ;[1 1 1] Target While Leaping - Start a timer that when ticks projects a capsule in the direction of the player's velocity applying effects to any actors that are hit by the trace.
  (bind _outhits (Collision|MultiCapsuleTraceForObjects (Transformation|GetActorLocation (Variables|Default|GetPlayerPawn)) (+ (Transformation|GetActorLocation (Variables|Default|GetPlayerPawn)) (* (Transformation|GetVelocity (Variables|Default|GetPlayerPawn)) (* (Utilities|Time|GetWorldDeltaSeconds) (ScalableFloat|GetValueatLevel (Variables|PowerTargetWhileLeaping|GetRow_WhileLeapingTraceFrames))))) (ScalableFloat|GetValueatLevel (Variables|PowerTargetWhileLeaping|GetRow_LeapingTargetCapsuleRadius)) (ScalableFloat|GetValueatLevel (Variables|PowerTargetWhileLeaping|GetRow_LeapingTargetCapsuleHalfHeight)) (Variables|PowerTargetWhileLeaping|GetWhileLeapingTargetObjectTypes) false (Variables|PowerTargetWhileLeaping|GetWhileLeapingHitActors) (Variables|PowerTargetWhileLeaping|GetWhileLeapingDrawDebug)))
  (CallFunction|ApplyEffecttoWhileLeapingTargets _outhits))

(event EventCastMontageNotifyStart
  (params (NotifyTag :type Gameplay%20Tag%20Structure))
  ;[1 1 1] Event Montage Notify Start Do the commit then Leap.
  (|CallParentFunction NotifyTag)
  (if (GameplayTags|Equal(GameplayTag) NotifyTag (Variables|PowerLeap|GetTag_LeapNotify))
    (Utilities|PowerCommitAbility (Variables|AbilityCommit|GetIgnoreAbilityCommit)
      (:Pass
        _)
      (:Fail
        (Development|PrintString "Power: Ability Commit Failed")
        (CallFunction|PowerCommitFailed))
      (:Ignored
        _))
    (CallFunction|Leap)
    (elif (and (GameplayTags|Equal(GameplayTag) NotifyTag (Variables|PowerAnimationCast|GetT_AbilityEndCastNotify)) (GameplayTags|NotEqual(GameplayTag) :A (Variables|PowerAnimationCast|GetT_AbilityEndCastNotify)))
      ;;[1 1 1] Unblock Player Movement if it was being blocked for the land
      (if (Variables|PowerLeap|GetBlockPlayerMovementForLand)
        (CallFunction|SetBlockMovement (Variables|Default|GetPlayerPawn) false (Variables|PowerLeap|GetBlockMovementOnLandHasDuration) (Variables|PowerLeap|GetBlockMovementOnLandDuration)) ; Unblock Player movement
        (else
          _))
      (CallFunction|CastMontageNotifyEndAbility))))

(event Ability|EventOnEndAbility
  (params (bWasCancelled :type Boolean))
  ;[1 1 1] On End Ability - Clear the While Leaping Target Timer.
  (|CallParentFunction bWasCancelled)
  (Utilities|Time|ClearandInvalidateTimerbyHandle (Variables|PowerTargetWhileLeaping|GetTimer_TargetWhileLeaping))
  ;[1 1 1] Failsafe to help ensure we unblock. Unblock Player Movement if it was being blocked for the land
  (if (Variables|PowerLeap|GetBlockPlayerMovementForLand)
    (CallFunction|SetBlockMovement (Variables|Default|GetPlayerPawn)) ; Unblock Player movement
    (else
      _))
  ;[1 1 1] Failsafe to ensure the loop cue is removed when the ability ends
  (if (Variables|PowerGameplayCues|GetForceRemoveLeapingLoopOnEnd)
    (Ability|RemoveGameplayCueFromOwner (Variables|PowerGameplayCues|GetGC_LeapingLoop))))

(event Custom|Target
  ;[1 1 1] Do Target And Apply Effect - Play burst GC, do target selection
  (CallFunction|PlayBurstGC (Variables|PowerGameplayCues|GetGC_Target) (Variables|PowerGameplayCues|GetTargetGCUsesParams) (CallFunction|TargetGCParams))
  (bind (_asynctaskproxy _targetdata _applicationtag) (Ability|Tasks|DoTargetSelection "Targeting" (Variables|PowerTargetingApply|GetTag_TargetingApplication) true)
    (:Targeted
      (CallFunction|ApplyEffect _targetdata _applicationtag (Variables|PowerTargetingApply|GetApplyEffectLevelTarget)))))

(event Custom|TargetLand
  ;[1 1 1] Do Targeting On Land And Apply Effect - Do target selection and apply effects. End Ability After
  (bind (_asynctaskproxy _targetdata _applicationtag) (Ability|Tasks|DoTargetSelection "TargetingLand" (Variables|PowerTargetingApply|GetTag_TargetingLandApplication) true)
    (:Targeted
      (CallFunction|ApplyEffect _targetdata _applicationtag (Variables|PowerTargetingApply|GetApplyEffectLevelTargetLand)))))

(event Custom|ApplyEffect
  (params (TargetData :type Gameplay%20Ability%20Target%20Data%20Handle%20Structure) (ApplicationTag :type Gameplay%20Tag%20Structure) (ApplyEffectLevel :type Integer))
  ;[1 1 1] Apply Effect to Target Data
  (Ability|ApplyGameplayEffectContainer TargetData ApplicationTag ApplyEffectLevel))

(fn ApplyEffectToWhileLeapingTargets
  (params (Array :type Array%20of%20Hit%20Result%20Structures))
  (bind _ge_whileleapinghit (Variables|PowerTargetWhileLeaping|GetGE_WhileLeapingHit))
  ;[1 1 1] Apply a gameplay effect to each hit actor
  (for _array_element Array
    (Utilities|IsValid (.HitActor (Collision|BreakHitResult _array_element))
      (:Is%20Valid
        (Utilities|Array|AddUnique (Variables|PowerTargetWhileLeaping|GetWhileLeapingHitActors) (.HitActor (Collision|BreakHitResult _array_element)))
        (if (Utilities|IsValidClass _ge_whileleapinghit)
          (Ability|ApplyGameplayEffectSpecToTarget (Ability|MakeOutgoingGameplayEffectSpec _ge_whileleapinghit (Variables|PowerTargetWhileLeaping|GetWhileLeapingEffectLevel)) (Ability|TargetData|AbilityTargetDatafromActor (.HitActor (Collision|BreakHitResult _array_element)))))))))

(fn CalculatePredictiveLaunchMultiplier
  (params (Targeting%20Hits :type Array%20of%20Hit%20Result%20Structures))
  ;[1 0.9175 0.3312] Calculate a multiplier for the launch velocity based on the nearest blocking hit
  (bind _returnvalue (CallFunction|GetLaunchVelocity))
  (LocalVariables|SetCachedPredictedLaunchDistance (* (Math|Vector|VectorLength _returnvalue) (ScalableFloat|GetValueatLevel (Variables|PowerLeap|PredictiveLaunchVelocity|GetRow_PredictiveLaunch_DistanceToCheckAsPctVelocity))))
  (for _array_element Targeting%20Hits
    (if (Default|ShouldComponentBlockLeap (.HitComponent (Collision|BreakHitResult _array_element)))
      (if (not (Default|WillTargetingHitDestroyActor (.HitActor (Collision|BreakHitResult _array_element))))
        (bind _output_get (LocalVariables|SetHitLaunchVelocityMultiplier (* (/ (Math|Vector|Distance(Vector) (.ImpactPoint (Collision|BreakHitResult _array_element)) (.TraceStart (Collision|BreakHitResult _array_element))) (LocalVariables|GetCachedPredictedLaunchDistance)) (ScalableFloat|GetValueatLevel (Variables|PowerLeap|PredictiveLaunchVelocity|GetRow_PredictiveLaunch_CollisionVelocityMult)))))
        (if (< _output_get (Variables|PowerLeap|PredictiveLaunchVelocity|RuntimeVars|GetPredictiveLaunchVelocityMult))
          (Variables|PowerLeap|PredictiveLaunchVelocity|RuntimeVars|SetPredictiveLaunchVelocityMult _output_get))))))

(fn CheckForEjection
  (Variables|PowerLeap|Ejection|RunTimeVars|SetEjectionLeap_ActorEjectedFrom 0)
  (Variables|PowerLeap|Ejection|RunTimeVars|SetWasEjected false)
  (if (ScalableFloat|GetValueasBool (Variables|PowerLeap|Ejection|GetRow_ShouldEjectFromOpenSeatVehicles))
    (bind _returnvalue (GameplayTags|HasMatchingGameplayTag(Message) (Variables|Default|GetPlayerPawn) "(TagName=\"Gameplay.Status.InVehicle.Seat.Open\")"))
    (if _returnvalue
      (Variables|PowerLeap|Ejection|RunTimeVars|SetLastVehicleVelocity (Vehicle|GetVehicleVelocity (Vehicle|GetVehicle (Variables|Default|GetPlayerPawn))))
      (Variables|PowerLeap|Ejection|RunTimeVars|SetEjectionLeap_ActorEjectedFrom (Vehicle|GetVehicle (Variables|Default|GetPlayerPawn)))
      (bind _returnvalue_2 (Vehicle|EjectPlayerPawn (Vehicle|GetVehicle (Variables|Default|GetPlayerPawn)) (Variables|Default|GetPlayerPawn)))
      (if _returnvalue_2
        _)
      (else
        (bind _returnvalue_3 (GameplayTags|HasMatchingGameplayTag(Message) (Variables|Default|GetPlayerPawn) "(TagName=\"Granted.Riding.Player.IsRiding\")"))
        (if _returnvalue_3
          (Variables|PowerLeap|Ejection|RunTimeVars|SetLastVehicleVelocity (Transformation|GetVelocity (Variables|Default|GetPlayerPawn)))
          (Variables|PowerLeap|Ejection|RunTimeVars|SetEjectionLeap_ActorEjectedFrom (Actor|GetAttachParentActor (Variables|Default|GetPlayerPawn))))))
    (Variables|PowerLeap|Ejection|RunTimeVars|SetWasEjected true)
    (else
      _)))

(fn EjectionLeap_GetClampVelocityOnFinish
  (return 0.0))

(fn EjectionLeap_GetDurationToIgnoreCollisionWithActorEjectedFrom
  (return (ScalableFloat|GetValueatLevel (Variables|PowerLeap|Ejection|GetRow_EjectionLeap_DefaultDurationToIgnoreCollisionWithActorEjectedFrom))))

(fn EjectionLeap_GetSetVelocityOnFinish
  (return "0, 0, 0"))

(fn GetAccelerationValue
  (params (Pawn :type Character%20Object%20Reference))
  ;[1 1 1] Get the normalized acceleration value. If set to assume forward on 0 accleration, use the forward vector in that case.
  (return (select (and (Variables|PowerLeap|UpwardOrDownward|GetZeroAccelerationAssumesForward) (Math|Vector|Equal(Vector) (Math|Vector|Normalize (Pawn|Components|CharacterMovement|GetCurrentAcceleration (Variables|/Script/Engine.Character|GetCharacterMovement Pawn))) "0, 0, 0" 0.0001)) (Transformation|GetActorForwardVector Pawn) (Math|Vector|Normalize (Pawn|Components|CharacterMovement|GetCurrentAcceleration (Variables|/Script/Engine.Character|GetCharacterMovement Pawn))))))

(fn GetArcCustomLocation
  ;[1 1 1] Function that is used to find the custom arc leap location. By default this returns 1000u above the player but should be overriden in children to do the custom calculations.
  (return (+ (Transformation|GetActorLocation (Variables|Default|GetPlayerPawn)) "0, 0,1000.000000")))

(fn GetArcTargetLocation
  (bind _playerpawn (Variables|Default|GetPlayerPawn))
  (bind (_outlocation _outrotation) (Actor|GetActorEyesViewPoint _playerpawn))
  (bind _arcleapdistanceusescharge (Variables|PowerLeap|Arc|GetArcLeapDistanceUsesCharge))
  (bind _row_arcleapdistance (Variables|PowerLeap|Arc|GetRow_ArcLeapDistance))
  (bind _returnvalue (CallFunction|GetRowatLevel _row_arcleapdistance))
  (bind _returnvalue_2 (CallFunction|ModifyLaunchMagwithCharge _returnvalue _arcleapdistanceusescharge))
  (bind _returnvalue_3 (Math|Vector|GetForwardVector _outrotation))
  (bind _returnvalue_4 (* _returnvalue_3 _returnvalue_2))
  ;[1 1 1] Do a trace to see where our arc leap is targeted to land. If we do not hit something, then launch the pawn a percentage of that distance instead.
  (bind (_outhit _returnvalue_5) (Collision|SphereTraceByChannel _outlocation (+ _outlocation _returnvalue_4) (Variables|PowerLeap|Arc|Trace|GetArcTraceRadius) (Variables|PowerLeap|Arc|Trace|GetArcLeapLocationTraceChannel) false (Utilities|Array|MakeArray (Variables|Default|GetPlayerPawn))))
  (return (Math|Vector|SelectVector (.ImpactPoint (Collision|BreakHitResult _outhit)) (+ _outlocation (* _returnvalue_4 (CallFunction|GetRowatLevel (Variables|PowerLeap|Arc|GetRow_ArcNoTargetLeapDistanceMultiplier)))) _returnvalue_5)))

(fn GetClampVelocityOnFinish
  ;[1 1 1] Get the clamp velocity on finish value. By default this is the "ClampVelOnFinish" variable but can be overriden.
  (return (Variables|PowerLeap|VelocityonFinish|GetClampVelOnFinish)))

(fn GetEjectionLeapVelocity
  (bind (_lastvehiclevelocity_x _lastvehiclevelocity_y _lastvehiclevelocity_z) (Math|Vector|BreakVector (Variables|GetLastVehicleVelocity_X)))
  (return (Math|Vector|MakeVector (* _lastvehiclevelocity_x (ScalableFloat|GetValueatLevel (Variables|PowerLeap|Ejection|GetRow_EjectionLeap_PctVehicleVelocityMaintained_Horizontal))) (* _lastvehiclevelocity_y (ScalableFloat|GetValueatLevel (Variables|PowerLeap|Ejection|GetRow_EjectionLeap_PctVehicleVelocityMaintained_Horizontal))) (+ (Math|Float|Max(Float) (* _lastvehiclevelocity_z (ScalableFloat|GetValueatLevel (Variables|PowerLeap|Ejection|GetRow_EjectionLeap_PctVehicltVelocityMaintained_Vertical)))) (ScalableFloat|GetValueatLevel (Variables|PowerLeap|Ejection|GetRow_EjectionLeap_AddedZVelocity))))))

(fn GetLaunchVelocity
  (bind (_returnvalue_x _returnvalue_y) (Math|Vector|BreakVector (Math|Vector|SelectVector (CallFunction|GetAccelerationValue (Variables|Default|GetPlayerPawn)) (Transformation|GetActorForwardVector (Variables|Default|GetPlayerPawn)) (Variables|PowerLeap|UpwardOrDownward|GetBaseLaunchDirectionOnAcceleration))))
  ;[1 1 1] Calculate the launch velocity based on the launch magnitudes and the direction based on wither forward vector or player input. Add in the velocity percent from the player's current velocity
  (return (* (+ (CallFunction|GetVelocityPercent) (Math|Vector|MakeVector (* _returnvalue_x (CallFunction|ModifyLaunchMagwithCharge (CallFunction|GetRowatLevel (Variables|PowerLeap|UpwardorDownward|GetRow_HorizontalLaunchMagnitude)) (Variables|PowerLeap|UpwardOrDownward|GetHorizontalLaunchMagnitudeUsesCharge))) (* _returnvalue_y (CallFunction|ModifyLaunchMagwithCharge (CallFunction|GetRowatLevel (Variables|PowerLeap|UpwardorDownward|GetRow_HorizontalLaunchMagnitude)) (Variables|PowerLeap|UpwardOrDownward|GetHorizontalLaunchMagnitudeUsesCharge))) (Utilities|Select :Index (Variables|PowerLeap|GetLeapType) :NewEnumerator0 0.0 :NewEnumerator3 0.0 :NewEnumerator2 (CallFunction|ModifyLaunchMagwithCharge (CallFunction|GetRowatLevel (Variables|PowerLeap|UpwardorDownward|GetRow_VerticalLaunchMagnitude)) (Variables|PowerLeap|UpwardOrDownward|GetVerticalLaunchMagnitudeUsesCharge)) :NewEnumerator1 (* (CallFunction|ModifyLaunchMagwithCharge (CallFunction|GetRowatLevel (Variables|PowerLeap|UpwardorDownward|GetRow_VerticalLaunchMagnitude)) (Variables|PowerLeap|UpwardOrDownward|GetVerticalLaunchMagnitudeUsesCharge)) -1.0)))) (Variables|PowerLeap|PredictiveLaunchVelocity|RuntimeVars|GetPredictiveLaunchVelocityMult))))

(fn GetLeapDuration
  ;[1 1 1] Gets the Duration of the leap ability and returns it. (Exists to override if needed. By Default uses Row Leap Duration)
  (return (ScalableFloat|GetValueatLevel (Variables|PowerLeap|GetRow_LeapDuration))))

(fn GetSetVelocityOnFinish
  ;[1 1 1] Get the set velocity on finish value. By default this is the "SetVelOnFinish" variable but can be overriden.
  (return (Variables|PowerLeap|VelocityonFinish|GetSetVelOnFinish)))

(fn GetVelocityPercent
  (bind (_velocity_x _velocity_y _velocity_z) (Math|Vector|BreakVector (Variables|/Script/Engine.MovementComponent|GetVelocity_X (Pawn|GetMovementComponent (Variables|Default|GetPlayerPawn)))))
  ;[1 1 1] Calc how much velocity is maintained from the players current velocity and return it.
  (return (Math|Vector|MakeVector (* _velocity_x (CallFunction|GetRowatLevel (Variables|PowerLeap|UpwardorDownward|GetRow_HorizontalVelocityPercentMaintained))) (* _velocity_y (CallFunction|GetRowatLevel (Variables|PowerLeap|UpwardorDownward|GetRow_HorizontalVelocityPercentMaintained))) (select (or (and (Utilities|Select :Index (Variables|PowerLeap|GetLeapType) :NewEnumerator0 false :NewEnumerator3 false :NewEnumerator2 (> (* _velocity_z (CallFunction|GetRowatLevel (Variables|PowerLeap|UpwardorDownward|GetRow_VertitcalVelocityPercentMaintained))) 0.0) :NewEnumerator1 (< (* _velocity_z (CallFunction|GetRowatLevel (Variables|PowerLeap|UpwardorDownward|GetRow_VertitcalVelocityPercentMaintained))) 0.0)) (Variables|PowerLeap|UpwardorDownward|GetVerticalVelocityMustMatchImpulse)) (not (Variables|PowerLeap|UpwardorDownward|GetVerticalVelocityMustMatchImpulse))) (* _velocity_z (CallFunction|GetRowatLevel (Variables|PowerLeap|UpwardorDownward|GetRow_VertitcalVelocityPercentMaintained))) 0.0))))

(fn GiveTillLandFallImmunity
  ;[1 1 1] Give Fall Damage Immunity till next land if the "GiveNextLandFallImmunityOnLeap" bool is true.
  (if (Variables|PowerFallImmunity|GetGiveNextLandFallImmunityOnLeap)
    (bind _returnvalue (Ability|ApplyGameplayEffectToOwner (Variables|PowerFallImmunity|GetGE_FallDamageImmunity) 0))
    (Variables|PowerFallImmunity|SetGEH_FallDamageImmunity _returnvalue)))

(fn InitAbility
  ;[1 1 1] On Init Reset the Do Onces, rest the while leaping hit actors
  (|CallParentFunction)
  (Utilities|Array|Clear (Variables|PowerTargetWhileLeaping|GetWhileLeapingHitActors))
  (CallFunction|ResetLandDoOnce)
  (Variables|PowerLeap|RuntimeVars|SetLaunchStarted false))

(fn LandGCParams
  (return 0))

(fn LeapGCParams
  (return 0))

(fn LeapingLoopGCParams
  (return 0))

(fn ModifyLaunchMagWithCharge
  (params (Mag :type Float%20%28double-precision%29) (Modify :type Boolean))
  ;[1 1 1] Modify Launch Magnitudes with Charge If Needed
  (return (Math|Float|SelectFloat (* Mag (Variables|PowerCharge|GetChargeMultiplier)) Mag Modify)))

(fn OnLaunchStarted)

(fn SelectArcLocation
  ;[1 1 1] Chose the location for the arc based on its type.
  (switch Utilities|FlowControl|Switch|SwitchonAthena_Power_Leap_Type (Variables|PowerLeap|GetLeapType)
    (:NewEnumerator0
      (bind _returnvalue (CallFunction|GetArcTargetLocation))
      (LocalVariables|SetLocation _returnvalue))
    (:NewEnumerator3
      (bind _location (CallFunction|GetArcCustomLocation))
      (LocalVariables|SetLocation _location)))
  (return (LocalVariables|GetLocation)))

(fn SetFallSettings
  (bind _playerpawn (Variables|Default|GetPlayerPawn))
  ;[1 1 1] Check if we ignore next fall and or reset fall height.
  (if (Variables|PowerLeap|GetOnLeapSetIgnoreNextFall)
    (Variables|/Script/FortniteGame.FortPawn|SetbIgnoreNextFallingDamage true _playerpawn)
    (if (Variables|PowerLeap|GetOnLeapResetFallHeight)
      (Pawn|Character|ResetFallingHeight _playerpawn))))

(fn ShouldComponentBlockLeap
  (params (Component :type Primitive%20Component%20Object%20Reference))
  (Utilities|IsValid Component
    (:Is%20Valid
      (return (and (Utilities|Array|ContainsItem (Variables|PowerLeap|PredictiveLaunchVelocity|GetObjectTypes_BlockLeapMovement) (Collision|GetCollisionObjectType Component)) (Utilities|Enum|Equal(Enum) (Collision|GetCollisionResponsetoChannel Component "ECC_Pawn") "ECR_Block"))))
    (:Is%20Not%20Valid
      (return false))))

(fn TargetGCParams
  (return 0))

(fn WillTargetingHitDestroyActor
  (params (HitActor :type Actor%20Object%20Reference))
  (Utilities|FlowControl|Sequence
    (:then_0
      (Utilities|IsValid HitActor
        (:Is%20Valid
          (if (Variables|/Script/Engine.Actor|GetbCanBeDamaged HitActor)
            (bind _asbuilding_actor (Utilities|Casting|CastToBuildingActor HitActor))
            (return (<= (Health|GetHealth _asbuilding_actor :!class /Script/FortniteGame.BuildingActor) (ScalableFloat|GetValueatLevel (Variables|PowerLeap|PredictiveLaunchVelocity|GetRow_TargetWhileLeaping_EnvDamage))))))))
    (:then_1
      (return false))))
