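; Leap active power movement mode extension for Fortnite Athena, built on the FortMovementMode base extension logic.
; Before each character movement update on the autonomous proxy, it checks whether the owner pawn has the ActivePower.Leap gameplay tag.
; If the tag is missing it returns Ended, otherwise it stays Active.
; Activation and deactivation only print debug strings, and the file sets no movement values itself.
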(event EventOnExtensionActivate
  (params (RuntimeContext :type MMERuntime%20Context%20Structure%20%28by%20ref%29))
  (Development|PrintString (Variables|FortMovementMode_BaseExtLogic|GetOwnerPawn) "Player just activated MME_Leap" false))

(event EventOnExtensionDeactivate
  (params (RuntimeContext :type MMERuntime%20Context%20Structure%20%28by%20ref%29))
  (Development|PrintString (Variables|FortMovementMode_BaseExtLogic|GetOwnerPawn) "player just ended MME_Leap"))

(fn UpdateBeforeCharacterMovement
  (params (ElapsedTime :type Float%20%28single-precision%29) (DeltaTime :type Float%20%28single-precision%29))
  (if (Utilities|Enum|Equal(Enum) (Networking|GetLocalRole (Variables|FortMovementMode_BaseExtLogic|GetOwnerPawn)) "ROLE_AutonomousProxy")
    (if (not (GameplayTags|HasMatchingGameplayTag (Variables|FortMovementMode_BaseExtLogic|GetOwnerPawn) "(TagName=\"Gameplay.Athena.ActivePower.Leap\")"))
      (Development|PrintString (Variables|FortMovementMode_BaseExtLogic|GetOwnerPawn) "MME_Leap force End because has no tag")
      (return "Ended")))
  (return "Active"))
